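# Physics asset for the F_MED_TutSoothe medium female player cosmetic body, covering a coarse body and a two-loop bow.
# Capsules on pelvis, upper spine, upper arms, thighs and calves form the body, with the bow base constrained to spine_05.
# Each loop base is angularly locked to the bow and leads a two-link dyn chain with tighter swing and twist limits.
# Left and right dyn loop capsules differ in length and orientation.
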PhysicsAsset /BRCosmetics/Characters/Player/Female/Medium/Bodies/F_MED_TutSoothe/Meshes/F_MED_TutSoothe_Physics
bodies {
 base_bow [{
   shapeType=Capsule
   length=2.6561
   radius=2.4775
   shapeName=base_bow_capsule
   center=<-0.733 -0.0076 -0.0014>
   rotation=<rot 78.5929 50.393 140.9534>
  }]
 base_bow_loop_01_l [{shapeType=Capsule, length=0.505, radius=0.505, shapeName=base_bow_loop_01_l_capsule, rotation=<rot 0 -90 0>}]
 base_bow_loop_01_r [{shapeType=Capsule, length=0.505, radius=0.505, shapeName=base_bow_loop_01_r_capsule, rotation=<rot 0 -90 0>}]
 calf_l [{shapeType=Capsule, length=24.5502, radius=9.9968, center=<-19.5996 -2.3117 -1.1974>, rotation=<rot -0.026 -89.2657 -89.5976>}]
 calf_r [{shapeType=Capsule, length=24.5502, radius=9.9968, center=<19.5996 2.3117 1.1974>, rotation=<rot -0.026 -89.2657 90.4024>}]
 dyn_bow_loop_01_l [{shapeType=Capsule, length=7.5, radius=2, shapeName=dyn_bow_loop_01_l_capsule, center=<2.6355 0 0.9296>}]
 dyn_bow_loop_01_r [{shapeType=Capsule, length=1, radius=2, shapeName=dyn_bow_loop_01_r_capsule, center=<0.5 0 0>, rotation=<rot -90 0.0004 0>}]
 dyn_bow_loop_02_l [{shapeType=Capsule, length=10, radius=2.7, shapeName=dyn_bow_loop_02_l_capsule, center=<3.3647 0 0>}]
 dyn_bow_loop_02_r [{shapeType=Capsule, length=2, radius=2.7, shapeName=dyn_bow_loop_02_r_capsule, center=<2.1501 0 0>, rotation=<rot -90 0 0>}]
 pelvis [{shapeType=Capsule, length=19, radius=10, center=<-7.4361 -0.6482 0.2231>}]
 spine_05 [{shapeType=Capsule, length=10, radius=9}]
 thigh_l [{shapeType=Capsule, length=23.0009, radius=7.8174, center=<-19.0373 -0.3734 -0.2465>, rotation=<rot 0.0106 -92.6904 -88.3429>}]
 thigh_r [{shapeType=Capsule, length=23.0009, radius=7.8174, center=<19.0373 0.3734 0.2465>, rotation=<rot 0.0106 -92.6904 91.6571>}]
 upperarm_l [{shapeType=Capsule, length=14.74, radius=5.4861, center=<12.3141 0.5437 -0.3767>, rotation=<rot 0 -90.2086 -87.7695>}]
 upperarm_r [{shapeType=Capsule, length=14.74, radius=5.4861, center=<-12.3141 -0.5437 0.3767>, rotation=<rot 0 -90.2086 92.2305>}]
}
constraints [
 {bone1Name=spine_05, bone2Name=pelvis, swing1LimitDegrees=45, swing2LimitDegrees=45, twistLimitDegrees=45}
 {bone1Name=upperarm_l, bone2Name=spine_05, swing1LimitDegrees=45, swing2LimitDegrees=45, twistLimitDegrees=45}
 {bone1Name=upperarm_r, bone2Name=spine_05, swing1LimitDegrees=45, swing2LimitDegrees=45, twistLimitDegrees=45}
 {bone1Name=thigh_l, bone2Name=pelvis, swing1LimitDegrees=45, swing2LimitDegrees=45, twistLimitDegrees=45}
 {bone1Name=calf_l, bone2Name=thigh_l, swing1LimitDegrees=45, swing2LimitDegrees=45, twistLimitDegrees=45}
 {bone1Name=thigh_r, bone2Name=pelvis, swing1LimitDegrees=45, swing2LimitDegrees=45, twistLimitDegrees=45}
 {bone1Name=calf_r, bone2Name=thigh_r, swing1LimitDegrees=45, swing2LimitDegrees=45, twistLimitDegrees=45}
 {bone1Name=dyn_bow_loop_01_l, bone2Name=base_bow_loop_01_l, swing1LimitDegrees=30, swing2LimitDegrees=30, twistLimitDegrees=5}
 {bone1Name=dyn_bow_loop_02_l, bone2Name=dyn_bow_loop_01_l, swing1LimitDegrees=30, swing2LimitDegrees=30, twistLimitDegrees=5}
 {bone1Name=dyn_bow_loop_01_r, bone2Name=base_bow_loop_01_r, swing1LimitDegrees=30, swing2LimitDegrees=30, twistLimitDegrees=5}
 {bone1Name=dyn_bow_loop_02_r, bone2Name=dyn_bow_loop_01_r, swing1LimitDegrees=30, swing2LimitDegrees=30, twistLimitDegrees=5}
 {
  bone1Name=base_bow_loop_01_l
  bone2Name=base_bow
  swing1LimitDegrees=45
  swing1Motion=Locked
  swing2LimitDegrees=45
  swing2Motion=Locked
  twistLimitDegrees=45
  twistMotion=Locked
 }
 {
  bone1Name=base_bow_loop_01_r
  bone2Name=base_bow
  swing1LimitDegrees=45
  swing1Motion=Locked
  swing2LimitDegrees=45
  swing2Motion=Locked
  twistLimitDegrees=45
  twistMotion=Locked
 }
 {bone1Name=base_bow, bone2Name=spine_05, swing1LimitDegrees=45, swing2LimitDegrees=45, twistLimitDegrees=45}
]
