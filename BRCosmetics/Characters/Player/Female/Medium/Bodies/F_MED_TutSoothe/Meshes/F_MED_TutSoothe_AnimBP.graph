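# Post-process-style cosmetic AnimBP for the F_MED_TutSoothe female body, adding secondary motion to its dangling parts.
# Copies the attached parent's pose, runs it through a Control Rig, then a clavicle_l-space RigidBody simulation.
# Cached result passes through five chained spring-driven Anim Dynamics nodes for pumpkin, ears, little face and left leg wrap.
# Each update, a nativized parameter-asset function computes and stores gravity override and RigidBody settings.
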AnimBlueprint /BRCosmetics/Characters/Player/Female/Medium/Bodies/F_MED_TutSoothe/Meshes/F_MED_TutSoothe_AnimBP
target_skeleton=&/F_MED_TutSoothe_Skeleton
graphs [
 AnimGraph {nodes [
   n1 = Anim Dynamics - Bone: dyn_pumpkin!(ComponentPose=Local To Component(LocalPose=Use cached pose 'RBDCache'()), Alpha=1, GravityOverride:Vector="0,0,-980", ExternalForce="0,0,0") -> (Pose) {
    angularDampingOverride=1
    angularSpringConstant=20000
    bAngularSpring=true
    bOverrideAngularDamping=true
    bUseGravityOverride=true
    gravityScale=0.35
    simulationSpace=RootRelative
    boundBone {boneName=dyn_pumpkin}
    componentLinearAccScale=<1 1 1>
    componentLinearVelScale=<1 1 1>
    gravityOverride {z=-980}
    physicsBodyDefinitions [{
      boundBone {boneName=dyn_pumpkin}
      boxExtents=<30 40 40>
      constraintSetup {angularLimitsMax=<15 15 15>, angularLimitsMin=<-15 -15 -15>, angularTarget=<1 0 0>}
      localJointOffset=<3 0 0>
     }]
   }
   n2 = Anim Dynamics - Bone: dyn_pumpkin_ear_01_r!(ComponentPose=@n1, Alpha=1, GravityOverride:Vector="0,0,-980", ExternalForce="0,0,0") -> (Pose) {
    angularDampingOverride=1
    angularSpringConstant=20000
    bAngularSpring=true
    bOverrideAngularDamping=true
    bUseGravityOverride=true
    gravityScale=0.35
    simulationSpace=BoneRelative
    boundBone {boneName=dyn_pumpkin_ear_01_r}
    componentLinearAccScale=<1 1 1>
    componentLinearVelScale=<1 1 1>
    gravityOverride {z=-980}
    physicsBodyDefinitions [{
      boundBone {boneName=dyn_pumpkin_ear_01_r}
      boxExtents=<19 28 28>
      constraintSetup {twistAxis=AxisZ, angularLimitsMax=<15 15 15>, angularLimitsMin=<-15 -15 -15>, angularTarget=<1 0 0>}
      localJointOffset=<4 0 0>
     }]
   }
   n3 = Anim Dynamics - Bone: dyn_pumpkin_ear_01_l!(ComponentPose=@n2, Alpha=1, GravityOverride:Vector="0,0,-980", ExternalForce="0,0,0") -> (Pose) {
    angularDampingOverride=1
    angularSpringConstant=20000
    bAngularSpring=true
    bOverrideAngularDamping=true
    bUseGravityOverride=true
    gravityScale=0.35
    simulationSpace=BoneRelative
    boundBone {boneName=dyn_pumpkin_ear_01_l}
    componentLinearAccScale=<1 1 1>
    componentLinearVelScale=<1 1 1>
    gravityOverride {z=-980}
    physicsBodyDefinitions [{
      boundBone {boneName=dyn_pumpkin_ear_01_l}
      boxExtents=<25 35 35>
      constraintSetup {twistAxis=AxisZ, angularLimitsMax=<15 15 15>, angularLimitsMin=<-15 -15 -15>, angularTarget=<1 0 0>}
      localJointOffset=<4.3 0 0>
     }]
   }
   n4 = Anim Dynamics - Bone: dyn_little_face!(ComponentPose=@n3, Alpha=1, GravityOverride:Vector="0,0,-980", ExternalForce="0,0,0") -> (Pose) {
    angularDampingOverride=1
    angularSpringConstant=20000
    bAngularSpring=true
    bOverrideAngularDamping=true
    bUseGravityOverride=true
    gravityScale=0.35
    simulationSpace=RootRelative
    boundBone {boneName=dyn_little_face}
    componentLinearAccScale=<1 1 1>
    componentLinearVelScale=<1 1 1>
    gravityOverride {z=-980}
    physicsBodyDefinitions [{
      boundBone {boneName=dyn_little_face}
      boxExtents=<20 30 30>
      constraintSetup {angularLimitsMax=<15 15 15>, angularLimitsMin=<-15 -15 -15>, angularTarget=<1 0 0>}
      localJointOffset=<4 0 0>
     }]
   }
   n5 = Anim Dynamics - Bone: dyn_leg_wrap_03_l!(ComponentPose=@n4, Alpha=1, GravityOverride:Vector="0,0,-980", ExternalForce="0,0,0") -> (Pose) {
    angularDampingOverride=1
    angularSpringConstant=20000
    bAngularSpring=true
    bOverrideAngularDamping=true
    bUseGravityOverride=true
    gravityScale=0.35
    simulationSpace=RootRelative
    boundBone {boneName=dyn_leg_wrap_03_l}
    componentLinearAccScale=<1 1 1>
    componentLinearVelScale=<1 1 1>
    gravityOverride {z=-980}
    physicsBodyDefinitions [{
      boundBone {boneName=dyn_leg_wrap_03_l}
      boxExtents=<20 30 30>
      constraintSetup {angularLimitsMax=<15 25 25>, angularLimitsMin=<-15 -10 0>, angularTarget=<1 0 0>}
      localJointOffset=<4 0 0>
     }]
   }
   Output Pose(Result=Component To Local(ComponentPose=@n5)) {layerGroup=None}
   n6 = Control Rig(Source=Copy Pose From Mesh() {bUseAttachedParent=true}) {controlRigAssetReference {controlRigAsset=&/F_MED_TutSoothe_CtrlRig}}
   Save cached pose 'ControlRigCache'(Pose=@n6) {cacheName=ControlRigCache}
   n7 = RigidBody!(
    ComponentPose=Local To Component(LocalPose=Use cached pose 'ControlRigCache'())
    Alpha
    OverrideWorldGravity:Vector="0,0,0"
    ExternalForce
    SimSpaceSettings:Sim Space Settings Structure=(WorldAlpha=0,VelocityScaleZ=1,DampingAlpha=1,MaxLinearVelocity=10000,MaxAngularVelocity=10000,MaxLinearAcceleration=10000,MaxAngularAcceleration=10000,ExternalLinearDragV=(X=0,Y=0,Z=0),ExternalLinearVelocity=(X=0,Y=0,Z=0),ExternalAngularVelocity=(X=0,Y=0,Z=0))
   ) -> (Pose) {
    bOverrideWorldGravity=true
    overridePhysicsAsset=&/F_MED_TutSoothe_Physics
    simulationSpace=BaseBoneSpace
    baseBoneRef {boneName=clavicle_l}
   }
   Save cached pose 'RBDCache'(Pose=Component To Local(ComponentPose=@n7)) {cacheName=RBDCache}
  ]}
 EventGraph {nodes [
   Event BlueprintUpdateAnimation()
   > Sequence!(execute) -> (then_0, then_1, then_2:Exec, then_3:Exec, then_4:Exec, then_5:Exec)
   > n1 = "Gravity Override State Function (Parameter Asset) (Nativized)"(
    CharacterPartAnimInstance=Self-Reference()
    GameTimeInSeconds=GetGameTimeInSeconds()
    Parameters=&/F_MED_TutSoothe_Parameters
   )
   > Set Gravity Override(Gravity Override=@n1)
   > Set RBAN Settings Out(RBAN Settings Out=@n1.OutRigidBodyAnimNodeParameters)
  ]}
]
variables [
 Gravity Override {type=Vector}
 RBAN Settings Out {
  type=FortRigidBodyAnimNodeParameters
  default {simSpaceSettings {
    dampingAlpha=0
    maxAngularAcceleration=25
    maxAngularVelocity=7.5
    maxLinearAcceleration=2000
    maxLinearVelocity=500
    velocityScaleZ=0.175
    worldAlpha=0.8
    externalLinearDragV=<0.2 0.2 0.2>
   }}
 }
]
