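# Post-copy secondary-motion AnimBP for the F_MED_TutSoothe face accessory, simulating ears and left-side earrings.
# Pose copied from the attached parent runs through head-relative AnimDynamics on three bones per ear, with right-ear spherical limits referencing left-ear bones.
# Constraints pin earring bases to left-ear bones before root-relative AnimDynamics swings the earrings.
# EventGraph sets gravity and RBAN variables from a parameter asset, but AnimGraph nodes use fixed gravity.
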AnimBlueprint /BRCosmetics/Characters/Player/Female/Medium/Bodies/F_MED_TutSoothe/Meshes/Parts/F_MED_TutSoothe_FaceAcc_AnimBP
target_skeleton=&/F_MED_TutSoothe_FaceAcc_Skeleton
graphs [
 AnimGraph {nodes [
   n1 = Anim Dynamics - Bone: dyn_ear_01_l!(
    ComponentPose=Local To Component(LocalPose=Copy Pose From Mesh() {bUseAttachedParent=true})
    Alpha=1
    GravityOverride:Vector="0,0,-980"
    ExternalForce="0,0,0"
   ) -> (Pose) {
    angularDampingOverride=1
    angularSpringConstant=20000
    bAngularSpring=true
    bOverrideAngularDamping=true
    bUseGravityOverride=true
    gravityScale=0.35
    simulationSpace=BoneRelative
    boundBone {boneName=dyn_ear_01_l}
    componentLinearAccScale=<1 1 1>
    componentLinearVelScale=<1 1 1>
    gravityOverride {z=-980}
    physicsBodyDefinitions [{
      collisionType=CustomSphere
      sphereCollisionRadius=5
      boundBone {boneName=dyn_ear_01_l}
      boxExtents=<55 40 40>
      constraintSetup {angularLimitsMax=<0 4 5>, angularLimitsMin=<0 -18 -5>, angularTarget=<1 0 0>}
      localJointOffset=<4 0 0>
     }]
    relativeSpaceBone {boneName=head}
    simSpaceSettings {maxAngularAcceleration=1000, maxAngularVelocity=5000, simSpaceAngularAlpha=0.1, externalAngularVelocity=<0.5 0.5 0.5>}
   }
   n2 = Anim Dynamics - Bone: dyn_ear_02_l!(ComponentPose=@n1, Alpha=1, GravityOverride:Vector="0,0,-980", ExternalForce="0,0,0") -> (Pose) {
    angularDampingOverride=1
    angularSpringConstant=20000
    bAngularSpring=true
    bOverrideAngularDamping=true
    bUseGravityOverride=true
    gravityScale=0.35
    simulationSpace=BoneRelative
    boundBone {boneName=dyn_ear_02_l}
    componentLinearAccScale=<1 1 1>
    componentLinearVelScale=<1 1 1>
    gravityOverride {z=-980}
    physicsBodyDefinitions [{
      collisionType=CustomSphere
      sphereCollisionRadius=5
      boundBone {boneName=dyn_ear_02_l}
      boxExtents=<55 40 40>
      constraintSetup {angularLimitsMax=<0 17 5>, angularLimitsMin=<0 -7 -17>, angularTarget=<1 0 0>}
      localJointOffset=<4 0 0>
     }]
    relativeSpaceBone {boneName=head}
    simSpaceSettings {maxAngularAcceleration=1000, maxAngularVelocity=5000, simSpaceAngularAlpha=0.1, externalAngularVelocity=<0.5 0.5 0.5>}
   }
   n3 = Anim Dynamics - Bone: dyn_ear_03_l!(ComponentPose=@n2, Alpha=1, GravityOverride:Vector="0,0,-980", ExternalForce="0,0,0") -> (Pose) {
    angularDampingOverride=1
    angularSpringConstant=20000
    bAngularSpring=true
    bOverrideAngularDamping=true
    bUseGravityOverride=true
    gravityScale=0.35
    simulationSpace=BoneRelative
    boundBone {boneName=dyn_ear_03_l}
    componentLinearAccScale=<1 1 1>
    componentLinearVelScale=<1 1 1>
    gravityOverride {z=-980}
    physicsBodyDefinitions [{
      collisionType=CustomSphere
      sphereCollisionRadius=5
      boundBone {boneName=dyn_ear_03_l}
      boxExtents=<55 40 40>
      constraintSetup {angularLimitsMax=<0 20 5>, angularLimitsMin=<0 -5 -15>, angularTarget=<1 0 0>}
      localJointOffset=<4 0 0>
     }]
    relativeSpaceBone {boneName=head}
    simSpaceSettings {maxAngularAcceleration=1000, maxAngularVelocity=5000, simSpaceAngularAlpha=0.1, externalAngularVelocity=<0.5 0.5 0.5>}
   }
   n4 = Anim Dynamics - Bone: dyn_ear_01_r!(ComponentPose=@n3, Alpha=1, GravityOverride:Vector="0,0,-980", ExternalForce="0,0,0") -> (Pose) {
    angularDampingOverride=1
    angularSpringConstant=20000
    bAngularSpring=true
    bOverrideAngularDamping=true
    bUseGravityOverride=true
    bUsePlanarLimit=false
    bUseSphericalLimits=true
    gravityScale=0.35
    simulationSpace=BoneRelative
    boundBone {boneName=dyn_ear_01_r}
    componentLinearAccScale=<0.7 0.7 0.7>
    componentLinearVelScale=<0.7 0.7 0.7>
    gravityOverride {z=-980}
    physicsBodyDefinitions [{
      collisionType=CustomSphere
      sphereCollisionRadius=5
      boundBone {boneName=dyn_ear_01_r}
      boxExtents=<51 35 35>
      constraintSetup {angularLimitsMax=<0 25 10>, angularLimitsMin=<0 -9 -5>, angularTarget=<1 0 0>}
      localJointOffset=<-4 0 0>
     }]
    relativeSpaceBone {boneName=head}
    simSpaceSettings {maxAngularAcceleration=1000, maxAngularVelocity=5000, simSpaceAngularAlpha=0.1, externalAngularVelocity=<0.5 0.5 0.5>}
    sphericalLimits [{limitRadius=0, drivingBone {boneName=None}, sphereLocalOffset=<0 0 0>}]
   }
   n5 = Anim Dynamics - Bone: dyn_ear_02_r!(ComponentPose=@n4, Alpha=1, GravityOverride:Vector="0,0,-980", ExternalForce="0,0,0") -> (Pose) {
    angularDampingOverride=1
    angularSpringConstant=20000
    bAngularSpring=true
    bOverrideAngularDamping=true
    bUseGravityOverride=true
    bUsePlanarLimit=false
    bUseSphericalLimits=true
    gravityScale=0.35
    simulationSpace=BoneRelative
    boundBone {boneName=dyn_ear_02_r}
    componentLinearAccScale=<0.7 0.7 0.7>
    componentLinearVelScale=<0.7 0.7 0.7>
    gravityOverride {z=-980}
    physicsBodyDefinitions [{
      collisionType=CustomSphere
      sphereCollisionRadius=5
      boundBone {boneName=dyn_ear_02_r}
      boxExtents=<45 35 35>
      constraintSetup {angularLimitsMax=<0 17 15>, angularLimitsMin=<0 -7 -5>, angularTarget=<1 0 0>}
      localJointOffset=<-4 0 0>
     }]
    relativeSpaceBone {boneName=head}
    simSpaceSettings {maxAngularAcceleration=1000, maxAngularVelocity=5000, simSpaceAngularAlpha=0.1, externalAngularVelocity=<0.5 0.5 0.5>}
    sphericalLimits [
     {limitRadius=4.1, drivingBone {boneName=dyn_ear_03_l}, sphereLocalOffset=<1 1 1>}
     {limitRadius=4.1, drivingBone {boneName=dyn_ear_02_l}, sphereLocalOffset=<0 0 0>}
    ]
   }
   n6 = Anim Dynamics - Bone: dyn_ear_03_r!(ComponentPose=@n5, Alpha=1, GravityOverride:Vector="0,0,-980", ExternalForce="0,0,0") -> (Pose) {
    angularDampingOverride=1
    angularSpringConstant=20000
    bAngularSpring=true
    bOverrideAngularDamping=true
    bUseGravityOverride=true
    bUsePlanarLimit=false
    bUseSphericalLimits=true
    gravityScale=0.35
    simulationSpace=BoneRelative
    boundBone {boneName=dyn_ear_03_r}
    componentLinearAccScale=<0.7 0.7 0.7>
    componentLinearVelScale=<0.7 0.7 0.7>
    gravityOverride {z=-980}
    physicsBodyDefinitions [{
      collisionType=CustomSphere
      sphereCollisionRadius=5
      boundBone {boneName=dyn_ear_03_r}
      boxExtents=<45 35 35>
      constraintSetup {angularLimitsMax=<0 20 15>, angularLimitsMin=<0 -5 -5>, angularTarget=<1 0 0>}
      localJointOffset=<-4 0 0>
     }]
    relativeSpaceBone {boneName=head}
    simSpaceSettings {maxAngularAcceleration=1000, maxAngularVelocity=5000, simSpaceAngularAlpha=0.1, externalAngularVelocity=<0.5 0.5 0.5>}
    sphericalLimits [
     {limitRadius=4.1, drivingBone {boneName=dyn_ear_03_l}, sphereLocalOffset=<0 0 0>}
     {limitRadius=4.1, drivingBone {boneName=dyn_ear_02_l}, sphereLocalOffset=<0 0 0>}
    ]
   }
   n7 = "Constraint - base_earring_lower_01_l (PP)"!(ComponentPose=@n6, Alpha=1, ConstraintWeights_0=0.9632, ConstraintWeights_1:float=0.0368) -> (Pose) {
    boneToModify {boneName=base_earring_lower_01_l}
    constraintSetup [{targetBone {boneName=dyn_ear_03_l}}, {targetBone {boneName=dyn_ear_02_l}}]
    constraintWeights [1,1]
   }
   n8 = "Constraint - base_earring_upper_01_l (P)"!(ComponentPose=@n7, Alpha=1, ConstraintWeights_0=1, ConstraintWeights_1:float=0) -> (Pose) {boneToModify {boneName=base_earring_upper_01_l}, constraintSetup [{targetBone {boneName=dyn_ear_03_l}}], constraintWeights [1,0]}
   n9 = Anim Dynamics - Bone: dyn_earring_upper_01_l!(ComponentPose=@n8, Alpha=1, GravityOverride:Vector="0,0,-980", ExternalForce="0,0,0") -> (Pose) {
    angularDampingOverride=1
    angularSpringConstant=10000
    bAngularSpring=true
    bOverrideAngularDamping=true
    bUseGravityOverride=true
    gravityScale=0.35
    simulationSpace=RootRelative
    boundBone {boneName=dyn_earring_upper_01_l}
    componentLinearAccScale=<0.8 0.8 0.8>
    componentLinearVelScale=<0.8 0.8 0.8>
    gravityOverride {z=-980}
    physicsBodyDefinitions [{
      boundBone {boneName=dyn_earring_upper_01_l}
      boxExtents=<10 20 20>
      constraintSetup {angularLimitsMax=<0 45 0>, angularLimitsMin=<0 -30 0>, angularTarget=<1 0 0>}
      localJointOffset=<5 0 0>
     }]
    simSpaceSettings {maxAngularAcceleration=1000, maxAngularVelocity=5000, simSpaceAngularAlpha=0.1, externalAngularVelocity=<0.5 0.5 0.5>}
   }
   n10 = Anim Dynamics - Bone: dyn_earring_lower_01_l!(ComponentPose=@n9, Alpha=1, GravityOverride:Vector="0,0,-980", ExternalForce="0,0,0") -> (Pose) {
    angularDampingOverride=1
    angularSpringConstant=20000
    bAngularSpring=true
    bOverrideAngularDamping=true
    bUseGravityOverride=true
    gravityScale=0.35
    simulationSpace=RootRelative
    boundBone {boneName=dyn_earring_lower_01_l}
    componentLinearAccScale=<1 1 1>
    componentLinearVelScale=<1 1 1>
    gravityOverride {z=-980}
    physicsBodyDefinitions [{
      boundBone {boneName=dyn_earring_lower_01_l}
      boxExtents=<20 30 30>
      constraintSetup {angularLimitsMax=<0 29 0>, angularLimitsMin=<0 -28 0>, angularTarget=<1 0 0>}
      localJointOffset=<7 0 0>
     }]
   }
   Output Pose(Result=Component To Local(ComponentPose=@n10)) {layerGroup=None}
  ]}
 EventGraph {nodes [
   Event BlueprintUpdateAnimation()
   > n1 = "Gravity Override State Function (Parameter Asset) (Nativized)"(
    CharacterPartAnimInstance=Self-Reference()
    GameTimeInSeconds=GetGameTimeInSeconds()
    Parameters=/BRCosmetics/Characters/Player/Female/Medium/Bodies/F_MED_TutSoothe/Meshes/F_MED_TutSoothe_Parameters
   )
   > Set Gravity Override(Gravity Override=@n1)
   > Set RBAN Settings Out(RBAN Settings Out=@n1.OutRigidBodyAnimNodeParameters)
  ]}
]
variables [
 Gravity Override {type=Vector}
 RBAN Settings Out {type=FortRigidBodyAnimNodeParameters}
]
